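# Sound cue for the Storm King boss's active crawling lightning bolt, routed to the boss fight SFX sound class.
# A random node picks evenly among three lightning summon impact waves.
# A modulator then raises the pitch within a narrow random range, reusing the impact sounds at a higher tone.
# Overridden attenuation sets spatial falloff, and overridden concurrency caps simultaneous instances, refusing new plays once the limit is reached.
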SoundCue /Game/Sounds/Fort_Monster_Storm_King/SFX/StormKing_CrawlingBolt_Active_Cue
root=Modulator_1
nodes [
 {id=Modulator_1, children [Random_1], properties {pitchMax=1.3, pitchMin=1.2}}
 {id=Random_1, children [WavePlayer_2, WavePlayer_3, WavePlayer_13], properties {weights [1,1,1]}}
 {id=WavePlayer_13, properties {soundWaveAssetPtr=&/StormKing_LightningSummons_Impact_02}}
 {id=WavePlayer_2, properties {soundWaveAssetPtr=&/StormKing_LightningSummons_Impact_03}}
 {id=WavePlayer_3, properties {soundWaveAssetPtr=&/StormKing_LightningSummons_Impact_01}}
]
properties {
 bOverrideAttenuation=true
 bOverrideConcurrency=true
 maxDistance=7000
 soundClassObject=&Fort_BossFight_SFX
 volumeMultiplier=1.1
 attenuationOverrides {falloffDistance=5000, attenuationShapeExtents {x=2000}}
 concurrencyOverrides {maxCount=3, resolutionRule=PreventNew}
}
