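# Sound cue for the start of the Storm King monster's death in its boss fight.
# A single wave player feeds one death-start wave directly to the output, slightly boosted, under the boss fight player-critical sound class.
# Cue-level attenuation overrides give it a very long audible range.
# A concurrency override allows only one instance at a time and blocks new plays rather than cutting off the current one.
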SoundCue /Game/Sounds/Fort_Monster_Storm_King/SFX/StormKing_Death_Start_Cue
root=WavePlayer_3
nodes [{id=WavePlayer_3, properties {soundWaveAssetPtr=/Game/Sounds/Fort_Monster_Storm_King/SFX/Final_SFX/StormKing_Death_v3_Start_01}}]
properties {
 bOverrideAttenuation=true
 bOverrideConcurrency=true
 maxDistance=30000
 soundClassObject=/Game/Sounds/FortSoundClassesAndModes/Fort_BossFight_SoundClass/Fort_BossFight_PlayerCritical
 volumeMultiplier=1.25
 attenuationOverrides {falloffDistance=20000, attenuationShapeExtents {x=10000}}
 concurrencyOverrides {maxCount=1, resolutionRule=PreventNew}
}
