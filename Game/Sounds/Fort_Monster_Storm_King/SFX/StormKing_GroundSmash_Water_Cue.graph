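# Sound cue for the water impact of the Storm King monster's ground smash, in the boss fight SFX class.
# A Random node picks one of four equally weighted melee water impact waves and feeds it through a Modulator for pitch and volume variation.
# Its own attenuation override makes it audible over a very long range, and its concurrency override prevents new instances once the limit is reached.
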SoundCue /Game/Sounds/Fort_Monster_Storm_King/SFX/StormKing_GroundSmash_Water_Cue
root=Modulator_3
nodes [
 {id=Modulator_3, children [Random_1]}
 {id=Random_1, children [WavePlayer_3,WavePlayer_4,WavePlayer_5,WavePlayer_6], properties {weights [1,1,1,1]}}
 {id=WavePlayer_3, properties {soundWaveAssetPtr=&/StormKing_Melee_Water_Impact_01}}
 {id=WavePlayer_4, properties {soundWaveAssetPtr=&/StormKing_Melee_Water_Impact_02}}
 {id=WavePlayer_5, properties {soundWaveAssetPtr=&/StormKing_Melee_Water_Impact_03}}
 {id=WavePlayer_6, properties {soundWaveAssetPtr=&/StormKing_Melee_Water_Impact_04}}
]
properties {
 bOverrideAttenuation=true
 bOverrideConcurrency=true
 maxDistance=30000
 soundClassObject=&Fort_BossFight_SFX
 volumeMultiplier=1
 attenuationOverrides {falloffDistance=20000, attenuationShapeExtents {x=10000}}
 concurrencyOverrides {resolutionRule=PreventNew}
}
