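# Sound cue for the Storm King boss's left-side lightning summon, layering a charge sound with a lightning impact.
# A mixer combines a random pick of three left charge waves with a slightly quieter, briefly randomly delayed pick of four impact waves, then feeds a modulator.
# It uses the boss-fight SFX class, long-range attenuation overrides, and concurrency that prevents new instances at the limit.
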SoundCue /Game/Sounds/Fort_Monster_Storm_King/SFX/StormKing_LightningSummons_Left_Cue
root=Modulator_3
nodes [
 {id=Delay_1, children [Random_4], properties {delayMax=0.2, delayMin=0.1}}
 {id=Mixer_2, children [Random_0,Delay_1], properties {inputVolume [1,0.7]}}
 {id=Modulator_3, children [Mixer_2]}
 {id=Random_0, children [WavePlayer_0,WavePlayer_1,WavePlayer_2], properties {weights [1,1,1]}}
 {id=Random_4, children [WavePlayer_3, WavePlayer_4, WavePlayer_18, WavePlayer_19], properties {weights [1,1,1,1]}}
 {id=WavePlayer_0, properties {soundWaveAssetPtr=/Game/Sounds/Fort_Monster_Storm_King/SFX/Final_SFX/StormKing_LightningSummons_Charge_Left_01}}
 {id=WavePlayer_1, properties {soundWaveAssetPtr=/Game/Sounds/Fort_Monster_Storm_King/SFX/Final_SFX/StormKing_LightningSummons_Charge_Left_02}}
 {
  id=WavePlayer_18
  properties {soundWaveAssetPtr=/Game/Sounds/Fort_Monster_Storm_King/SFX/Final_SFX/StormKing_LightningSummons_Lightning_Impact_03}
 }
 {
  id=WavePlayer_19
  properties {soundWaveAssetPtr=/Game/Sounds/Fort_Monster_Storm_King/SFX/Final_SFX/StormKing_LightningSummons_Lightning_Impact_04}
 }
 {id=WavePlayer_2, properties {soundWaveAssetPtr=/Game/Sounds/Fort_Monster_Storm_King/SFX/Final_SFX/StormKing_LightningSummons_Charge_Left_03}}
 {
  id=WavePlayer_3
  properties {soundWaveAssetPtr=/Game/Sounds/Fort_Monster_Storm_King/SFX/Final_SFX/StormKing_LightningSummons_Lightning_Impact_01}
 }
 {
  id=WavePlayer_4
  properties {soundWaveAssetPtr=/Game/Sounds/Fort_Monster_Storm_King/SFX/Final_SFX/StormKing_LightningSummons_Lightning_Impact_02}
 }
]
properties {
 bOverrideAttenuation=true
 bOverrideConcurrency=true
 maxDistance=30000
 soundClassObject=&Fort_BossFight_SFX
 volumeMultiplier=1
 attenuationOverrides {falloffDistance=20000, attenuationShapeExtents {x=10000}}
 concurrencyOverrides {resolutionRule=PreventNew}
}
