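# Sound cue for the Storm King boss's multi-target projectile summon, routed to the boss fight SFX sound class.
# A Random node picks one of three equally weighted lightning summon waves and feeds a Modulator that pins volume, leaving pitch at node defaults.
# The cue overrides attenuation with its own long-range falloff so the summon carries far across the fight.
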SoundCue /Game/Sounds/Fort_Monster_Storm_King/SFX/StormKing_MultiTarget_Projectile_Summon_Cue
root=Modulator_3
nodes [
 {id=Modulator_3, children [Random_1], properties {volumeMax=1, volumeMin=1}}
 {id=Random_1, children [WavePlayer_2,WavePlayer_3,WavePlayer_4], properties {weights [1,1,1]}}
 {id=WavePlayer_2, properties {soundWaveAssetPtr=/Game/Sounds/Fort_Monster_Storm_King/SFX/Final_SFX/StormKing_Projectile_Lightning_Summon_01}}
 {id=WavePlayer_3, properties {soundWaveAssetPtr=/Game/Sounds/Fort_Monster_Storm_King/SFX/Final_SFX/StormKing_Projectile_Lightning_Summon_02}}
 {id=WavePlayer_4, properties {soundWaveAssetPtr=/Game/Sounds/Fort_Monster_Storm_King/SFX/Final_SFX/StormKing_Projectile_Lightning_Summon_03}}
]
properties {
 bOverrideAttenuation=true
 maxDistance=30000
 soundClassObject=&Fort_BossFight_SFX
 attenuationOverrides {falloffDistance=20000, attenuationShapeExtents {x=10000}}
 concurrencyOverrides {resolutionRule=PreventNew}
}
