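# Sound cue for the Storm King boss's projectile explosion, reusing the Storm King hammer overhead-swing boulder impact waves.
# A random node picks one of six equally weighted impact variations, which passes through a very short randomized delay and then a modulator for variation.
# It routes to the boss-fight projectile sound class with overridden attenuation and its own concurrency limit that prevents new plays when full.
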SoundCue /Game/Sounds/Fort_Monster_Storm_King/SFX/StormKing_Projectile_Explosion_Cue
root=Modulator_1
nodes [
 {id=Delay_1, children [Random_1], properties {delayMax=0.05}}
 {id=Modulator_1, children [Delay_1]}
 {id=Random_1, children [WavePlayer_0,WavePlayer_4,WavePlayer_5,WavePlayer_6,WavePlayer_7,WavePlayer_8], properties {weights [1,1,1,1,1,1]}}
 {
  id=WavePlayer_0
  properties {soundWaveAssetPtr=/Game/Sounds/Fort_Weapons/Storm_King/StormKing_Hammer/StormKing_Hammer_SwingOverhead_BouldersImpact_01}
 }
 {
  id=WavePlayer_4
  properties {soundWaveAssetPtr=/Game/Sounds/Fort_Weapons/Storm_King/StormKing_Hammer/StormKing_Hammer_SwingOverhead_BouldersImpact_02}
 }
 {
  id=WavePlayer_5
  properties {soundWaveAssetPtr=/Game/Sounds/Fort_Weapons/Storm_King/StormKing_Hammer/StormKing_Hammer_SwingOverhead_BouldersImpact_03}
 }
 {
  id=WavePlayer_6
  properties {soundWaveAssetPtr=/Game/Sounds/Fort_Weapons/Storm_King/StormKing_Hammer/StormKing_Hammer_SwingOverhead_BouldersImpact_04}
 }
 {
  id=WavePlayer_7
  properties {soundWaveAssetPtr=/Game/Sounds/Fort_Weapons/Storm_King/StormKing_Hammer/StormKing_Hammer_SwingOverhead_BouldersImpact_05}
 }
 {
  id=WavePlayer_8
  properties {soundWaveAssetPtr=/Game/Sounds/Fort_Weapons/Storm_King/StormKing_Hammer/StormKing_Hammer_SwingOverhead_BouldersImpact_06}
 }
]
properties {
 attenuationSettings=/Game/Sounds/Attenuation/Fort_Bullet_Impact_Boss_Att
 bOverrideAttenuation=true
 bOverrideConcurrency=true
 maxDistance=4200
 soundClassObject=/Game/Sounds/FortSoundClassesAndModes/Fort_BossFight_SoundClass/Fort_BossFight_Projectile
 volumeMultiplier=1.2
 attenuationOverrides {attenuationShapeExtents {x=600}}
 concurrencyOverrides {maxCount=4, resolutionRule=PreventNew}
}
